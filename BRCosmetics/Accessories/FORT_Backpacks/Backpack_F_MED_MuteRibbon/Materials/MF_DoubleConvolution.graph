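# Material function in the Mute Ribbon backpack cosmetic's folder that builds a soft edge-style mask from four scalar samples.
# It takes the absolute differences of Sample1/Sample2 and Sample3/Sample4 and keeps the larger, similar to a Roberts cross edge detector.
# An inverse lerp from Threshold to Threshold plus Attenuation remaps that difference, which is saturated into the Mask output.
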MaterialFunction /BRCosmetics/Accessories/FORT_Backpacks/Backpack_F_MED_MuteRibbon/Materials/MF_DoubleConvolution
graph {nodes [
  n1 = FunctionInput() {bUsePreviewValueAsDefault=true, inputName=Threshold, inputType=FunctionInput_Scalar, sortPriority=5, previewValue=<0.01 0 0 1>}
  n2 = FunctionInput() {bUsePreviewValueAsDefault=true, inputName=Attenuation, inputType=FunctionInput_Scalar, sortPriority=6, previewValue=<0.01 0 0 1>}
  n3 = Add(A=@n1, B=@n2)
  n4 = ComponentMask(None=FunctionInput() {bUsePreviewValueAsDefault=true, inputName=Sample1, inputType=FunctionInput_Scalar, sortPriority=1}) {r=true}
  n5 = ComponentMask(None=FunctionInput() {bUsePreviewValueAsDefault=true, inputName=Sample2, inputType=FunctionInput_Scalar, sortPriority=2}) {r=true}
  n6 = ComponentMask(None=FunctionInput() {bUsePreviewValueAsDefault=true, inputName=Sample3, inputType=FunctionInput_Scalar, sortPriority=3}) {r=true}
  n7 = ComponentMask(None=FunctionInput() {bUsePreviewValueAsDefault=true, inputName=Sample4, inputType=FunctionInput_Scalar, sortPriority=4}) {r=true}
  n8 = Max(A=Abs(None=Subtract(A=@n4, B=@n5)), B=Abs(None=Subtract(A=@n6, B=@n7)))
  FunctionOutput(None=Saturate(None=InverseLinearInterpolate(A=@n1, B=@n3, Value=@n8))) {outputName=Mask}
 ]}
