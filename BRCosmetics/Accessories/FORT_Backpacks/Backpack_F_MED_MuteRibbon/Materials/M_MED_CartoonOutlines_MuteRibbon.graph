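# Translucent, unlit, two-sided outline material for the MuteRibbon backpack that draws cartoon ink lines through an expanded shell.
# World position offset pushes vertices along normals, with screen-space or world-space thickness varying by camera distance.
# Screen-space mode convolves neighboring roughness and scene-depth samples for line opacity, with a constant fallback on forward and mobile.
# Emissive uses metallic-detected white lines or an outline color with optional texture, passed through a LUT.
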Material /BRCosmetics/Accessories/FORT_Backpacks/Backpack_F_MED_MuteRibbon/Materials/M_MED_CartoonOutlines_MuteRibbon
graph {
 properties {
  bCastRayTracedShadows=false
  bUseMaterialAttributes=true
  bUsedWithMeshDeformer=true
  bUsedWithSkeletalMesh=true
  blendMode=BLEND_Translucent
  pixelDepthOffsetMode=PDOM_Legacy
  shadingModel=MSM_Unlit
  twoSided=true
  usedShadingModels=Unlit
 }
 nodes [
  n1 = SceneTexture(UVs=NamedRerouteUsage()) {sceneTextureId=PPI_Roughness}
  n2 = SceneTexture(UVs=NamedRerouteUsage()) {sceneTextureId=PPI_Roughness}
  n3 = SceneTexture(UVs=NamedRerouteUsage()) {sceneTextureId=PPI_Roughness}
  n4 = SceneTexture(UVs=NamedRerouteUsage()) {sceneTextureId=PPI_Roughness}
  n5 = ScalarParameter() {defaultValue=0.2, group=Outlines Roughness, parameterName=RoughnessThreshold, sliderMax=2, sortPriority=0}
  n6 = ScalarParameter() {defaultValue=0.05, group=Outlines Roughness, parameterName=RoughnessAttenuation, sliderMax=2, sortPriority=1}
  n7 = MaterialFunctionCall!(Sample1=@n1.Color, Sample2=@n2.Color, Sample3=@n3.Color, Sample4=@n4.Color, Threshold=@n5, Attenuation=@n6) {materialFunction=&/MF_DoubleConvolution}
  n8 = NamedRerouteUsage()
  n9 = MaterialFunctionCall!(
   Sample1=@n1.Color
   Sample2=@n2.Color
   Sample3=@n3.Color
   Sample4=@n4.Color
   BaseSample=SceneTexture() {sceneTextureId=PPI_Roughness}.Color
   BaseOffset
   Threshold=@n5
   Attenuation=@n6
  ) {materialFunction=&/MF_SingleConvolution}
  n10 = LinearInterpolate(A=Max(A=@n7.Mask, B=NamedRerouteUsage()), B=Multiply(A=OneMinus(None=@n8), B=@n9.Mask), Alpha=NamedRerouteUsage())
  n11 = LinearInterpolate(A=@n7.Mask, B=@n9.Mask, Alpha=NamedRerouteUsage())
  n12 = NamedRerouteUsage()
  n13 = NamedRerouteUsage()
  n14 = StaticSwitchParameter(
   True=StaticSwitchParameter(
    True=StaticSwitchParameter(True=@n10, False=@n11) {group=Outlines Depth, parameterName=DepthDebug, sortPriority=2}
    False=StaticSwitchParameter(True=@n12, False=@n13) {group=Outlines Depth, parameterName=DepthDebug, sortPriority=2}
   ) {group=Outlines Roughness, parameterName=RoughnessDebug, sortPriority=2}
   False=NamedRerouteUsage()
  ) {defaultValue=true, group=Outlines, parameterName=UseScreenSpaceOutlines, sortPriority=10}
  n15 = Constant() {r=1}
  n16 = NamedRerouteUsage()
  n17 = StaticSwitchParameter(True=@n15, False=StaticSwitchParameter(True=@n15, False=@n16) {group=Outlines Depth, parameterName=DepthDebug, sortPriority=2}) {group=Outlines Roughness, parameterName=RoughnessDebug, sortPriority=2}
  MaterialFunctionCall(In=SetMaterialAttributes!(
    MaterialAttributes
    Emissive Color=MaterialFunctionCall!(In=@n14, LUT, LUT Multiplier) {materialFunction=/Game/Packages/Fortress_SharedMaterials/Anime_Master_Material/MF_ApplyLUT}.Result
    Opacity=StaticSwitchParameter(True=@n17, False=@n16) {defaultValue=true, group=Outlines, parameterName=UseScreenSpaceOutlines, sortPriority=10}
    World Position Offset=NamedRerouteUsage()
   ) {attributeSetTypes [#1, #2, #3]}) {materialFunction=&MF_TechArtOverrides}
  n18 = TwoSidedSign()
  n19 = ComponentMask(None=ScreenPosition().ViewportUV) {g=true, r=true}
  n20 = Multiply(A=NamedRerouteUsage()) {constB=2}
  n21 = ComponentMask(None=@n20) {r=true}
  n22 = Constant()
  n23 = SceneTexture(UVs=Add(A=@n19, B=AppendVector(A=@n21, B=@n22)) {desc=x+}) {sceneTextureId=PPI_Metallic}
  Multiply(A=Saturate(None=@n18), B=@n23.Color)
  Multiply(A=@n18) {constB=-1}
  n24 = Subtract(A=PixelDepth().R, B=ScalarParameter() {defaultValue=250, group=Outlines, parameterName=NearDistance, sliderMax=3000, sortPriority=0})
  n25 = ScalarParameter() {defaultValue=600, group=Outlines, parameterName=FarDistance, sliderMax=3000, sortPriority=1}
  NamedRerouteDeclaration(None=Saturate(None=Divide(A=@n24, B=@n25))) {name=PixelDistance, nodeColor=<rgba 0.6641 0 1 1>}
  n26 = Constant()
  n27 = ShadingPathSwitch(Default=@n10, Forward=@n26, Mobile=@n26)
  n28 = Constant()
  n29 = Multiply(A=TwoSidedSign()) {constB=-1}
  NamedRerouteDeclaration(None=Max(
    A=StaticSwitchParameter(True=@n27, False=@n28) {defaultValue=true, group=Outlines, parameterName=UseScreenSpaceOutlines, sortPriority=10}
    B=Saturate(None=@n29)
   )) {name=opacity, nodeColor=<rgba 1 0 0.9141 1>}
  n30 = ScalarParameter() {defaultValue=2, group=Outlines, parameterName=ScreenSpaceThicknessNear, sliderMax=6, sliderMin=1, sortPriority=2}
  n31 = LinearInterpolate(
   A=Multiply(A=@n30) {constB=0.5}
   B=ScalarParameter() {defaultValue=1, group=Outlines, parameterName=ScreenSpaceThicknessFar, sliderMax=3, sliderMin=1, sortPriority=3}
   Alpha=NamedRerouteUsage()
  )
  n32 = ComponentMask(None=ViewProperty() {property=MEVP_ResolutionFraction}.Property) {r=true}
  NamedRerouteDeclaration(None=Multiply(A=SceneTexelSize(), B=Multiply(A=@n31, B=@n32) {constA=0.5})) {name=Thickness Screen Space, nodeColor=<rgba 0 0.6016 1 1>}
  n33 = SceneTexelSize()
  n34 = ScalarParameter() {defaultValue=2, group=Outlines, parameterName=ScreenSpaceThicknessNear, sliderMax=6, sliderMin=1, sortPriority=2}
  n35 = Multiply(
   A=LinearInterpolate(
    A=@n34
    B=ScalarParameter() {defaultValue=1, group=Outlines, parameterName=ScreenSpaceThicknessFar, sliderMax=3, sliderMin=1, sortPriority=3}
    Alpha=NamedRerouteUsage()
   )
   B=ComponentMask(None=ViewProperty() {property=MEVP_ResolutionFraction}.Property) {r=true}
  )
  n36 = WorldPosition()
  n37 = CameraPositionWS()
  n38 = ViewProperty() {property=MEVP_TanHalfFieldOfView}
  n39 = Multiply(
   A=ComponentMask(None=Multiply(A=@n33, B=@n35)) {r=true}
   B=Multiply(A=Distance(A=@n36.XYZ, B=@n37), B=ComponentMask(None=@n38.Property) {r=true})
  )
  n40 = LinearInterpolate(
   A=ScalarParameter() {defaultValue=0.4, group=Outlines, parameterName=WorldSpaceThicknessNear, sortPriority=2}
   B=ScalarParameter() {defaultValue=1, group=Outlines, parameterName=WorldSpaceThicknessFar, sortPriority=3}
   Alpha=NamedRerouteUsage()
  )
  NamedRerouteDeclaration(None=Multiply(
    A=StaticSwitchParameter(True=@n39, False=@n40) {defaultValue=true, group=Outlines, parameterName=UseScreenSpaceOutlines, sortPriority=10}
    B=VertexNormalWS()
   )) {name=wpo, nodeColor=<rgba 0.7578 1 0 1>}
  n41 = Subtract(A=Abs(None=Subtract(A=@n23.Color) {constB=0.04})) {constB=0.01}
  n42 = Multiply(A=@n21) {constB=-1}
  n43 = SceneTexture(UVs=Add(A=@n19, B=AppendVector(A=@n42, B=@n22)) {desc=y+}) {sceneTextureId=PPI_Metallic}
  n44 = Subtract(A=Abs(None=Subtract(A=@n43.Color) {constB=0.04})) {constB=0.01}
  n45 = ComponentMask(None=@n20) {g=true}
  n46 = SceneTexture(UVs=Add(A=@n19, B=AppendVector(A=@n22, B=@n45)) {desc=x-}) {sceneTextureId=PPI_Metallic}
  n47 = Subtract(A=Abs(None=Subtract(A=@n46.Color) {constB=0.04})) {constB=0.01}
  n48 = Multiply(A=@n45) {constB=-1}
  n49 = SceneTexture(UVs=Add(A=@n19, B=AppendVector(A=@n22, B=@n48)) {desc=y-}) {sceneTextureId=PPI_Metallic}
  n50 = Subtract(A=Abs(None=Subtract(A=@n49.Color) {constB=0.04})) {constB=0.01}
  n51 = Max(A=Max(A=Ceil(None=@n41), B=Ceil(None=@n44)), B=Max(A=Ceil(None=@n47), B=Ceil(None=@n50)))
  n52 = Constant()
  n53 = StaticSwitchParameter(
   True=Multiply(A=OneMinus(None=@n51), B=ScalarParameter() {defaultValue=0.3, group=MuteRibbonBackpack, parameterName=WhiteIntensity}) {constB=0.4}
   False=@n52
  ) {defaultValue=true, group=Outlines, parameterName=UseScreenSpaceOutlines, sortPriority=10}
  n54 = TextureSampleParameter2D() {
   group=Outlines
   parameterName=Color
   samplerSource=SSM_Clamp_WorldGroupSettings
   sortPriority=7
   texture=/BRCosmetics/Characters/Player/Female/Medium/Bodies/F_MED_MuteRibbon/Skins/Maroon/Textures/T_F_MED_MuteRibbon_Maroon_Head_DFColor
  }
  n55 = VectorParameter() {group=Outlines, parameterName=OutlineColor, sortPriority=-1, defaultValue=<rgba 0 0 0 1>}
  n56 = Multiply(A=Multiply(A=@n54.RGB), B=@n55.RGB)
  NamedRerouteDeclaration(None=StaticSwitchParameter(
    True=ShadingPathSwitch(Default=@n53, Forward=@n52, Mobile=@n52)
    False=StaticSwitchParameter(True=@n56, False=@n55.RGB) {group=Outlines, parameterName=UseColorTexture, sortPriority=6}
   ) {group=Outlines, parameterName=UseMetallicWhiteLines, sortPriority=5}) {name=emissive, nodeColor=<rgba 0.5938 0 1 1>}
  n57 = ComponentMask(None=ScreenPosition().ViewportUV) {g=true, r=true}
  n58 = NamedRerouteUsage()
  n59 = ComponentMask(None=@n58) {r=true}
  n60 = Constant()
  n61 = NamedRerouteDeclaration(None=Add(A=@n57, B=AppendVector(A=@n59, B=@n60)) {desc=x+}) {name=UV1, nodeColor=<rgba 1 0 0.375 1>}
  n62 = Multiply(A=@n59) {constB=-1}
  n63 = NamedRerouteDeclaration(None=Add(A=@n57, B=AppendVector(A=@n62, B=@n60)) {desc=y+}) {name=UV2, nodeColor=<rgba 0 0.6484 1 1>}
  n64 = ComponentMask(None=@n58) {g=true}
  n65 = NamedRerouteDeclaration(None=Add(A=@n57, B=AppendVector(A=@n60, B=@n64)) {desc=x-}) {name=UV3, nodeColor=<rgba 0.5 0 1 1>}
  n66 = Multiply(A=@n64) {constB=-1}
  n67 = NamedRerouteDeclaration(None=Add(A=@n57, B=AppendVector(A=@n60, B=@n66)) {desc=y-}) {name=UV4, nodeColor=<rgba 0.6641 1 0 1>}
  n68 = ScalarParameter() {defaultValue=4, group=Outlines Depth, parameterName=DepthThresholdNear, sortPriority=0}
  n69 = ScalarParameter() {defaultValue=10, group=Outlines Depth, parameterName=DepthThresholdFar, sliderMax=30, sortPriority=1}
  n70 = NamedRerouteUsage()
  NamedRerouteDeclaration(None=MaterialFunctionCall(
    Sample1=SceneDepth(UVs=@n61).R
    Sample2=SceneDepth(UVs=@n63).R
    Sample3=SceneDepth(UVs=@n65).R
    Sample4=SceneDepth(UVs=@n67).R
    BaseSample=PixelDepth().R
    BaseOffset=LinearInterpolate(A=@n68, B=@n69, Alpha=@n70) {constA=1, constB=10}
   ) {materialFunction=&MF_SingleConvolution}.RawMask) {name=Depth Lines, nodeColor=<rgba 0 0 0 1>}
 ]
}
