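# Physics asset for the arms of the MuteRibbon backpack cosmetic, defining collision bodies and joint limits for physical simulation.
# Capsule bodies cover a central pelvis and left and right clavicle, upper arm, lower arm and hand bones.
# Constraints chain each arm from hand up through clavicle to the pelvis, with tight swing limits and almost no twist.
# The upper arm joints get the widest swing, and limits differ slightly between left and right.
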PhysicsAsset /BRCosmetics/Accessories/FORT_Backpacks/Backpack_F_MED_MuteRibbon/Meshes/F_MED_MuteRibbon_Pack_Arms_Physics
bodies {
 clav_l [{shapeType=Capsule, length=3.3213, radius=1.8008, shapeName=clav_l_capsule, center=<-1.6151 0.1497 1.2114>, rotation=<rot 0 1.9718 -84.6834>}]
 clav_r [{shapeType=Capsule, length=3.4734, radius=1.6938, shapeName=clav_r_capsule, center=<1.8084 0.2431 -0.746>, rotation=<rot 0 -176.9518 -88.965>}]
 hand_l [{shapeType=Capsule, length=1.25, radius=1.25, shapeName=hand_l_capsule, center=<1 0 0>, rotation=<rot 90 0 0>}]
 hand_r [{shapeType=Capsule, length=1.25, radius=1.25, shapeName=hand_r_capsule, center=<-1 0 0>, rotation=<rot 90 0 0>}]
 lowerArm_l [{shapeType=Capsule, length=2, radius=1.25, shapeName=lowerArm_l_capsule, center=<1 0 0>, rotation=<rot 90 0 0>}]
 lowerArm_r [{shapeType=Capsule, length=2, radius=1.25, shapeName=lowerArm_r_capsule, center=<-1 0 0>, rotation=<rot 90 0 0>}]
 pelvis_c [{shapeType=Capsule, length=5, radius=6.2, shapeName=pelvis_c_capsule, center=<0.1 0 0>, rotation=<rot 90 0 0>}]
 upperArm_l [{shapeType=Capsule, length=4, radius=1.25, shapeName=upperArm_l_capsule, center=<2 0 0>, rotation=<rot 90 0 0>}]
 upperArm_r [{shapeType=Capsule, length=4, radius=1.25, shapeName=upperArm_r_capsule, center=<-2.6 0 0>, rotation=<rot 90 0 0>}]
}
constraints [
 {bone1Name=lowerArm_l, bone2Name=upperArm_l, swing1LimitDegrees=0, swing2LimitDegrees=7, twistLimitDegrees=0}
 {bone1Name=hand_l, bone2Name=lowerArm_l, swing1LimitDegrees=5, swing2LimitDegrees=10, twistLimitDegrees=0}
 {bone1Name=lowerArm_r, bone2Name=upperArm_r, swing1LimitDegrees=1, swing2LimitDegrees=5, twistLimitDegrees=0}
 {bone1Name=hand_r, bone2Name=lowerArm_r, swing1LimitDegrees=2, swing2LimitDegrees=7.5, twistLimitDegrees=0}
 {bone1Name=upperArm_l, bone2Name=clav_l, swing1LimitDegrees=20, swing2LimitDegrees=10, twistLimitDegrees=0}
 {bone1Name=upperArm_r, bone2Name=clav_r, swing1LimitDegrees=20, swing2LimitDegrees=5, twistLimitDegrees=0}
 {bone1Name=clav_l, bone2Name=pelvis_c, swing1LimitDegrees=5, swing2LimitDegrees=5, twistLimitDegrees=1}
 {bone1Name=clav_r, bone2Name=pelvis_c, swing1LimitDegrees=5, swing2LimitDegrees=5, twistLimitDegrees=1}
]
