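# Physics asset for the MuteRibbon backpack cosmetic, defining bodies and joints for a small figure with a cape, upper-eye bones and two legs.
# Every body is a single capsule, with a large one on the head and small ones elsewhere.
# Head, cape root and both legs attach to the pelvis, while the eye chain hangs off the head.
# Chain joints have tight swing and near-zero twist limits; head-to-pelvis and cape-root-to-pelvis rotate freely.
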PhysicsAsset /BRCosmetics/Accessories/FORT_Backpacks/Backpack_F_MED_MuteRibbon/Meshes/F_MED_MuteRibbon_Pack_Physics
bodies {
 cape_00_c [{shapeType=Capsule, length=2, radius=2, shapeName=cape_00_c_capsule, center=<1 0 0>, rotation=<rot 90 0 0>}]
 cape_01_c [{shapeType=Capsule, length=2, radius=2, shapeName=cape_01_c_capsule, center=<1 0 0>, rotation=<rot 90 0 0>}]
 cape_02_c [{shapeType=Capsule, length=2, radius=2, shapeName=cape_02_c_capsule, center=<1 0 0>, rotation=<rot 90 0 0>}]
 cape_03_c [{shapeType=Capsule, length=2, radius=2, shapeName=cape_03_c_capsule, center=<1 0 0>, rotation=<rot 90 0 0>}]
 dyn_upperEye_00_c [{shapeType=Capsule, length=1.25, radius=1.25, shapeName=dyn_upperEye_00_c_capsule, center=<1 0 0>, rotation=<rot 90 0 0>}]
 dyn_upperEye_01_c [{shapeType=Capsule, length=1.25, radius=1.25, shapeName=dyn_upperEye_01_c_capsule, center=<1 0 0>, rotation=<rot 90 0 0>}]
 foot_l [{shapeType=Capsule, length=1.25, radius=1.25, shapeName=foot_l_capsule, center=<1 0 0>, rotation=<rot 90 0 0>}]
 foot_r [{shapeType=Capsule, length=1.25, radius=1.25, shapeName=foot_r_capsule, center=<1 0 0>, rotation=<rot 90 0 0>}]
 head_c [{
   shapeType=Capsule
   length=10.1688
   radius=7.6674
   shapeName=head_c_capsule
   center=<0.0204 -0.6885 5.6275>
   rotation=<rot -0.0001 -165.6934 -89.0642>
  }]
 lowerLeg_l [{shapeType=Capsule, length=1.25, radius=1.25, shapeName=lowerLeg_l_capsule, center=<0.8 0 0>, rotation=<rot 90 0 0>}]
 lowerLeg_r [{shapeType=Capsule, length=1.25, radius=1.25, shapeName=lowerLeg_r_capsule, center=<1 0 0>, rotation=<rot 90 0 0>}]
 pelvis_c [{shapeType=Capsule, length=4, radius=2, shapeName=pelvis_c_capsule, center=<1 0 0>, rotation=<rot 0 0 90>}]
 upperLeg_l [{shapeType=Capsule, length=1.25, radius=1.25, shapeName=upperLeg_l_capsule, center=<0.75 0 0>, rotation=<rot 90 0 0>}]
 upperLeg_r [{shapeType=Capsule, length=1.25, radius=1.25, shapeName=upperLeg_r_capsule, center=<1.5 0 0>, rotation=<rot 90 0 0>}]
}
constraints [
 {bone1Name=cape_01_c, bone2Name=cape_00_c, swing1LimitDegrees=10, swing2LimitDegrees=4, twistLimitDegrees=0}
 {bone1Name=cape_02_c, bone2Name=cape_01_c, swing1LimitDegrees=10, swing2LimitDegrees=4, twistLimitDegrees=0}
 {bone1Name=cape_03_c, bone2Name=cape_02_c, swing1LimitDegrees=10, swing2LimitDegrees=6, twistLimitDegrees=0}
 {bone1Name=dyn_upperEye_00_c, bone2Name=head_c, swing1LimitDegrees=5, swing2LimitDegrees=5, twistLimitDegrees=1}
 {bone1Name=dyn_upperEye_01_c, bone2Name=dyn_upperEye_00_c, swing1LimitDegrees=5, swing2LimitDegrees=12, twistLimitDegrees=1}
 {bone1Name=lowerLeg_l, bone2Name=upperLeg_l, swing1LimitDegrees=1, swing2LimitDegrees=10, twistLimitDegrees=1}
 {bone1Name=foot_l, bone2Name=lowerLeg_l, swing1LimitDegrees=10, swing2LimitDegrees=10, twistLimitDegrees=1}
 {
  bone1Name=head_c
  bone2Name=pelvis_c
  swing1LimitDegrees=45
  swing1Motion=Free
  swing2LimitDegrees=45
  swing2Motion=Free
  twistLimitDegrees=45
  twistMotion=Free
 }
 {
  bone1Name=cape_00_c
  bone2Name=pelvis_c
  swing1LimitDegrees=45
  swing1Motion=Free
  swing2LimitDegrees=45
  swing2Motion=Free
  twistLimitDegrees=45
  twistMotion=Free
 }
 {bone1Name=upperLeg_l, bone2Name=pelvis_c, swing1LimitDegrees=20, swing2LimitDegrees=15, twistLimitDegrees=1}
 {bone1Name=upperLeg_r, bone2Name=pelvis_c, swing1LimitDegrees=20, swing2LimitDegrees=15, twistLimitDegrees=1}
 {bone1Name=lowerLeg_r, bone2Name=upperLeg_r, swing1LimitDegrees=1, swing2LimitDegrees=10, twistLimitDegrees=1}
 {bone1Name=foot_r, bone2Name=lowerLeg_r, swing1LimitDegrees=10, swing2LimitDegrees=10, twistLimitDegrees=1}
]
