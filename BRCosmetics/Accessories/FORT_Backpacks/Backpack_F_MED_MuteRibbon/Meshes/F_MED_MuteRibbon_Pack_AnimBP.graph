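# Animation Blueprint for the MuteRibbon backpack, adding physics motion, skydive posing and idle animation.
# A base pose feeds two chained RigidBody nodes (pack and arms physics assets), then additive bone offsets, some gated by In Air, with a weighted random idle/idle-break applied additively.
# Each update sets gravity overrides via a deprecated helper, and skydive states set In Air and lower inertia scales.
# The "Animation Testing" state machine is not connected to the output.
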AnimBlueprint /BRCosmetics/Accessories/FORT_Backpacks/Backpack_F_MED_MuteRibbon/Meshes/F_MED_MuteRibbon_Pack_AnimBP
target_skeleton=&/F_MED_MuteRibbon_Pack_Skeleton
graphs [
 AnimGraph {nodes [
   Input Pose() {name=InPose_0}
   n1 = Sequence Player 'MuteRibbon_BasePose_Anim_PACK'() {bLoopAnimation=false, sequence=/BRCosmetics/Accessories/FORT_Backpacks/Backpack_F_MED_MuteRibbon/Animation/MuteRibbon_BasePose_Anim_PACK}
   n2 = RigidBody!(
    ComponentPose=Local To Component(LocalPose=@n1)
    Alpha
    OverrideWorldGravity:Vector=Get Gravity Override()
    ComponentLinearAccScale:Vector=Get Component Linear Acc Scale() -> (Component Linear Acc Scale:Vector)
    ComponentLinearVelScale:Vector=Get Component Linear Vel Scale() -> (Component Linear Vel Scale:Vector)
   ) -> (Pose) {
    bClampLinearTranslationLimitToRefPose=true
    bOverrideWorldGravity=true
    overridePhysicsAsset=&/F_MED_MuteRibbon_Pack_Physics
    simulationSpace=BaseBoneSpace
    alphaBoolBlend {blendOption=Linear}
    baseBoneRef {boneName=pack}
    componentAppliedLinearAccClamp=<3000 3000 3000>
    componentLinearAccScale=<0.45 0.35 0.35>
    componentLinearVelScale=<0.45 0.35 0.35>
    overrideWorldGravity {z=-980}
    simSpaceSettings {
     maxAngularAcceleration=1000
     maxAngularVelocity=3
     maxLinearAcceleration=700
     maxLinearVelocity=500
     velocityScaleZ=0
     worldAlpha=0.5
     externalLinearDragV=<0.2 0.2>
    }
   }
   n3 = RigidBody(
    ComponentPose=@n2
    OverrideWorldGravity=Get Gravity Override_Arms() -> (Gravity Override_Arms:Vector)
    ComponentLinearAccScale=Get Component Linear Acc Scale_Arms() -> (Component Linear Acc Scale_Arms:Vector)
    ComponentLinearVelScale=Get Component Linear Vel Scale_Arms() -> (Component Linear Vel Scale_Arms:Vector)
   ) {
    bClampLinearTranslationLimitToRefPose=true
    bOverrideWorldGravity=true
    overridePhysicsAsset=&/F_MED_MuteRibbon_Pack_Arms_Physics
    simulationSpace=BaseBoneSpace
    alphaBoolBlend {blendOption=Linear}
    baseBoneRef {boneName=pack}
    componentAppliedLinearAccClamp=<3000 3000 3000>
    componentLinearAccScale=<0.45 0.35 0.35>
    componentLinearVelScale=<0.45 0.35 0.35>
    overrideWorldGravity {z=-980}
    simSpaceSettings {
     maxAngularAcceleration=1000
     maxAngularVelocity=3
     maxLinearAcceleration=700
     maxLinearVelocity=500
     velocityScaleZ=0
     worldAlpha=0.5
     externalLinearDragV=<0.2 0.2>
    }
   }
   n4 = "Transform (Modify) Bone - Bone: pack"(ComponentPose=@n3, Translation="0,-1.5,0", Rotation="0,0,-9") {
    rotationMode=BMM_Additive
    rotationSpace=BCS_BoneSpace
    translationMode=BMM_Additive
    translationSpace=BCS_BoneSpace
    alphaBoolBlend {blendOption=Linear}
    boneToModify {boneName=pack}
    rotation {roll=-9}
    translation {y=-1.5}
   }
   n5 = Get In Air() -> (In Air:float)
   n6 = "Transform (Modify) Bone - Bone: upperLeg_l"(ComponentPose=@n4, Alpha=@n5, Translation="0,0,0", Rotation="-10,0,0", Scale="1,1,1") {
    rotationMode=BMM_Additive
    rotationSpace=BCS_BoneSpace
    alphaBoolBlend {blendOption=Linear}
    alphaScaleBiasClamp {bInterpResult=true}
    boneToModify {boneName=upperLeg_l}
    rotation {pitch=-10}
   }
   n7 = "Transform (Modify) Bone - Bone: upperLeg_r"(ComponentPose=@n6, Alpha=@n5, Translation="0,0,0", Rotation="-10,0,0", Scale="1,1,1") {
    rotationMode=BMM_Additive
    rotationSpace=BCS_BoneSpace
    alphaBoolBlend {blendOption=Linear}
    alphaScaleBiasClamp {bInterpResult=true}
    boneToModify {boneName=upperLeg_r}
    rotation {pitch=-10}
   }
   n8 = "Transform (Modify) Bone - Bone: pack"(ComponentPose=@n7, Alpha=@n5, Translation="0,-2,-3.5", Rotation="0,0,12") {
    rotationMode=BMM_Additive
    rotationSpace=BCS_BoneSpace
    translationMode=BMM_Additive
    translationSpace=BCS_BoneSpace
    alphaBoolBlend {blendOption=Linear}
    alphaScaleBiasClamp {bInterpResult=true}
    boneToModify {boneName=pack}
    rotation {roll=12}
    translation {y=-2, z=-3.5}
   }
   n9 = "Transform (Modify) Bone - Bone: dyn_upperEye_00_c"(ComponentPose=@n8, Alpha=@n5, Translation="0,0,0", Rotation="12,0,0", Scale="1,1,1") {
    rotationMode=BMM_Additive
    rotationSpace=BCS_BoneSpace
    translationSpace=BCS_BoneSpace
    alphaBoolBlend {blendOption=Linear}
    alphaScaleBiasClamp {bInterpResult=true}
    boneToModify {boneName=dyn_upperEye_00_c}
    rotation {pitch=12}
   }
   n10 = "Transform (Modify) Bone - Bone: foot_l"(ComponentPose=@n9, Alpha=1, Translation="0,0,0", Rotation="30,0,0", Scale="1,1,1") {
    rotationMode=BMM_Additive
    rotationSpace=BCS_BoneSpace
    translationSpace=BCS_BoneSpace
    alphaBoolBlend {blendOption=Linear}
    alphaScaleBiasClamp {bInterpResult=true}
    boneToModify {boneName=foot_l}
    rotation {pitch=30}
   }
   n11 = "Transform (Modify) Bone - Bone: foot_r"(ComponentPose=@n10, Alpha=1, Translation="0,0,0", Rotation="30,0,0", Scale="1,1,1") {
    rotationMode=BMM_Additive
    rotationSpace=BCS_BoneSpace
    translationSpace=BCS_BoneSpace
    alphaBoolBlend {blendOption=Linear}
    alphaScaleBiasClamp {bInterpResult=true}
    boneToModify {boneName=foot_r}
    rotation {pitch=30}
   }
   n12 = "Transform (Modify) Bone - Bone: upperArm_r"(ComponentPose=@n11, Alpha=1, Translation="0,0,0", Rotation="0,5,0", Scale="1,1,1") {
    rotationMode=BMM_Additive
    rotationSpace=BCS_BoneSpace
    translationSpace=BCS_BoneSpace
    alphaBoolBlend {blendOption=Linear}
    alphaScaleBiasClamp {bInterpResult=true}
    boneToModify {boneName=upperArm_r}
    rotation {yaw=5}
   }
   n13 = "Transform (Modify) Bone - Bone: upperArm_l"(ComponentPose=@n12, Alpha=1, Translation="0,0,0", Rotation="0,10,0", Scale="1,1,1") {
    rotationMode=BMM_Additive
    rotationSpace=BCS_BoneSpace
    translationSpace=BCS_BoneSpace
    alphaBoolBlend {blendOption=Linear}
    alphaScaleBiasClamp {bInterpResult=true}
    boneToModify {boneName=upperArm_l}
    rotation {yaw=10}
   }
   n14 = "Transform (Modify) Bone - Bone: upperEye_c_FIX"(ComponentPose=@n13, Alpha=1, Translation="0,-0.1,0", Rotation="0,0,0", Scale="1,1,1") {
    rotationMode=BMM_Additive
    rotationSpace=BCS_BoneSpace
    translationMode=BMM_Additive
    translationSpace=BCS_BoneSpace
    alphaBoolBlend {blendOption=Linear}
    alphaScaleBiasClamp {bInterpResult=true}
    boneToModify {boneName=upperEye_c_FIX}
    translation {y=-0.1}
   }
   n15 = Random Sequence Player!() -> (Pose:Pose Link Structure) {
    is=title AnimGraphNode_RandomPlayer
    entries [
     {
      chanceToPlay=0.8
      maxLoopCount=0
      maxPlayRate=0.8
      minLoopCount=0
      minPlayRate=1
      sequence=/BRCosmetics/Accessories/FORT_Backpacks/Backpack_F_MED_MuteRibbon/Animation/MuteRibbon_Idle_Anim_PACK
      blendIn {blendOption=Linear, blendTime=0.2, customCurve=null}
     }
     {
      chanceToPlay=0.2
      maxLoopCount=0
      maxPlayRate=1
      minLoopCount=0
      minPlayRate=1
      sequence=/BRCosmetics/Accessories/FORT_Backpacks/Backpack_F_MED_MuteRibbon/Animation/MuteRibbon_IdleBreak_Anim_PACK
      blendIn {blendOption=Linear, blendTime=0.2, customCurve=null}
     }
    ]
   }
   Output Pose(Result=Apply Additive(Base=Component To Local(ComponentPose=@n14), Additive=@n15) {alphaBoolBlend {blendOption=Linear}}) {layerGroup=None}
   AnimGraphNode_StateMachine_0 = Animation Testing() {is=title AnimGraphNode_StateMachine}
  ]}
 Animation Testing {
  entry=Idle_1
  path=AnimGraph.AnimGraphNode_StateMachine_0.Animation Testing
  ids=id
  nodes [
   AnimStateNode_0 = Idle_1() {
    is=title AnimStateNode
    stateEntered {guid=#1}
    stateFullyBlended {guid=#2}
    stateLeft {guid=#3}
   }
   AnimStateNode_1 = Idle_2() {
    is=title AnimStateNode
    stateEntered {guid=#4}
    stateFullyBlended {guid=#5}
    stateLeft {guid=#6}
   }
   AnimStateNode_2 = Idle_Break() {
    is=title AnimStateNode
    stateEntered {guid=#7}
    stateFullyBlended {guid=#8}
    stateLeft {guid=#9}
   }
   AnimStateTransitionNode_2 = AnimStateTransitionNode() {
    bAutomaticRuleBasedOnSequencePlayerInState=true
    transitionEnd {guid=#10}
    transitionInterrupt {guid=#11}
    transitionStart {guid=#12}
   }
   AnimStateTransitionNode_3 = AnimStateTransitionNode() {
    bAutomaticRuleBasedOnSequencePlayerInState=true
    transitionEnd {guid=#13}
    transitionInterrupt {guid=#14}
    transitionStart {guid=#15}
   }
   AnimStateTransitionNode_4 = AnimStateTransitionNode() {bAutomaticRuleBasedOnSequencePlayerInState=true}
   AnimStateTransitionNode_5 = AnimStateTransitionNode() {bAutomaticRuleBasedOnSequencePlayerInState=true}
  ]
  links [
   {
    dst=Idle_2
    rule=AnimGraph.AnimGraphNode_StateMachine_0.Animation Testing.AnimStateTransitionNode_2.Transition
    src=Idle_1
    via=AnimStateTransitionNode_2
   }
   {
    dst=Idle_Break
    rule=AnimGraph.AnimGraphNode_StateMachine_0.Animation Testing.AnimStateTransitionNode_3.Transition
    src=Idle_2
    via=AnimStateTransitionNode_3
   }
   {
    dst=Idle_1
    rule=AnimGraph.AnimGraphNode_StateMachine_0.Animation Testing.AnimStateTransitionNode_4.Transition
    src=Idle_Break
    via=AnimStateTransitionNode_4
   }
   {
    dst=Idle_1
    rule=AnimGraph.AnimGraphNode_StateMachine_0.Animation Testing.AnimStateTransitionNode_5.Transition
    src=Idle_Break
    via=AnimStateTransitionNode_5
   }
  ]
 }
 Idle_1 {
  path=AnimGraph.AnimGraphNode_StateMachine_0.Animation Testing.AnimStateNode_0.Idle_1
  nodes [
   n1 = "Sequence Player (additive) 'MuteRibbon_Idle_Anim_PACK'"() {bLoopAnimation=false, sequence=&MuteRibbon_Idle_Anim_PACK}
   Output Animation Pose(Result=@n1)
  ]
 }
 Idle_2 {
  path=AnimGraph.AnimGraphNode_StateMachine_0.Animation Testing.AnimStateNode_1.Idle_2
  nodes [
   n1 = "Sequence Player (additive) 'MuteRibbon_Idle_Anim_PACK'"() {bLoopAnimation=false, sequence=&MuteRibbon_Idle_Anim_PACK}
   Output Animation Pose(Result=@n1)
  ]
 }
 Idle_Break {
  path=AnimGraph.AnimGraphNode_StateMachine_0.Animation Testing.AnimStateNode_2.Idle_Break
  nodes [
   n1 = "Sequence Player (additive) 'MuteRibbon_IdleBreak_Anim_PACK'"() {bLoopAnimation=false, sequence=&MuteRibbon_IdleBreak_Anim_PACK}
   Output Animation Pose(Result=@n1)
  ]
 }
 Transition {path=AnimGraph.AnimGraphNode_StateMachine_0.Animation Testing.AnimStateTransitionNode_2.Transition, nodes [Result()]}
 Transition {path=AnimGraph.AnimGraphNode_StateMachine_0.Animation Testing.AnimStateTransitionNode_3.Transition, nodes [Result()]}
 Transition {path=AnimGraph.AnimGraphNode_StateMachine_0.Animation Testing.AnimStateTransitionNode_4.Transition, nodes [Result()]}
 Transition {path=AnimGraph.AnimGraphNode_StateMachine_0.Animation Testing.AnimStateTransitionNode_5.Transition, nodes [Result()]}
 EventGraph {nodes [
   Event BlueprintUpdateAnimation()
   > n1 = Sequence!(execute) -> (then_0, then_1, then_2:Exec, then_3:Exec)
   n2 = GetGameTimeInSeconds()
   n3 = Self-Reference()
   > n4 = Gravity Override State Function DEPRECATED(
    CharacterPartAnimInstance=@n3
    GameTimeInSeconds=@n2
    GravityOverrideParams=Get Gravity Override Params_Arms() -> (Gravity Override Params_Arms:Gravity Override Params Struct Structure)
   )
   > Set Gravity Override_Arms(Gravity Override_Arms:Vector=@n4) -> (Output_Get:Vector)
   n5 = Gravity Override State Function DEPRECATED(execute=@n1.then_1, CharacterPartAnimInstance=@n3, GameTimeInSeconds=@n2, GravityOverrideParams=Get Gravity Override Params())
   > Set Gravity Override(Gravity Override=@n5)
   n6 = OR Boolean!(A=Get bIsSkyDiving(), B=Get bIsSkydiveDiveMode(), C:bool=Get bIsSkydiveFloating()) -> (ReturnValue)
   n7 = Branch(execute=@n1.then_2, Condition=@n6)
   > Set In Air(In Air:float=1) -> (Output_Get:float)
   Set In Air(execute=@n7.else, In Air=0)
   n8 = Branch(execute=@n1.then_3, Condition=@n6)
   > Set Component Linear Acc Scale(Component Linear Acc Scale:Vector="0.23,0.2,0.2") -> (Output_Get:Vector)
   > Set Component Linear Vel Scale(Component Linear Vel Scale:Vector="0.23,0.2,0.2") -> (Output_Get:Vector)
   > Set Component Linear Acc Scale_Arms(Component Linear Acc Scale_Arms:Vector="0.23,0.2,0.2") -> (Output_Get:Vector)
   > Set Component Linear Vel Scale_Arms(Component Linear Vel Scale_Arms:Vector="0.23,0.2,0.2") -> (Output_Get:Vector)
   Set Component Linear Acc Scale(execute=@n8.else, Component Linear Acc Scale="0.45,0.35,0.35")
   > Set Component Linear Vel Scale(Component Linear Vel Scale="0.45,0.35,0.35")
   > Set Component Linear Acc Scale_Arms(Component Linear Acc Scale_Arms="0.35,0.3,0.3")
   > Set Component Linear Vel Scale_Arms(Component Linear Vel Scale_Arms="0.35,0.3,0.3")
  ]}
]
variables [
 Component Linear Acc Scale {type=Vector, default=<0.45 0.35 0.35>}
 Component Linear Acc Scale_Arms {type=Vector, default=<0.45 0.35 0.35>}
 Component Linear Vel Scale {type=Vector, default=<0.45 0.35 0.35>}
 Component Linear Vel Scale_Arms {type=Vector, default=<0.45 0.35 0.35>}
 Gravity Override {type=Vector, default=<0 0 -980>}
 Gravity Override Params {type=GravityOverrideParamsStruct}
 Gravity Override Params_Arms {type=GravityOverrideParamsStruct}
 Gravity Override_Arms {type=Vector}
 In Air {type=number, default=1}
]
